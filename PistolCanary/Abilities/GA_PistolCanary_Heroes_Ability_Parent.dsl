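; Parent gameplay ability for PistolCanary hero supers, played by an AI-bot-controlled hero pawn.
; On activation it caches pawn and AI controller, commits, takes behavior control and stops firing, then plays the super montage.
; Montage notify tags call OnStartComicPanel and OnAssignDamage, which only debug-print here as hooks for child abilities.
; Montage completion, interruption, cancellation or absence ends the ability, and OnEndAbility pops behavior control from a valid controller.
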(event Ability|EventActivateAbility
  (bind (_montageinfo_montagetoplay _montageinfo_animplayrate _montageinfo_animrootmotiontranslationscale) (Utilities|Struct|BreakFortGameplayAbilityMontageInfo (Variables|GetMontageInfo_MontageToPlay)))
  (Development|PrintString "@@@ Event ActivateAbility [GA_PistolCanary_Heroes_Ability_Parent]: CALLED!" false (Variables|Debug|GetbDebugAbility))
  (CallFunction|InitHeroSuperAbility)
  (CallFunction|SetupHeroSuperAbility)
  (Utilities|IsValid _montageinfo_montagetoplay
    (:Is%20Valid
      (bind (_asynctaskproxy _notifytag) (Ability|Tasks|PlayMontageAndWaitNotify "PistolCanary_HeroSuper" _montageinfo_montagetoplay _montageinfo_animplayrate "None" true _montageinfo_animrootmotiontranslationscale)
        (:then
          (Variables|Internal|SetSuperMontageTask _asynctaskproxy))
        (:OnCompleted
          _)
        (:OnInterrupted
          _)
        (:OnCancelled
          _)
        (:OnNotifyBegin
          (switch Utilities|FlowControl|Switch|SwitchonGameplayTag _notifytag
            (:Event.HeroAbility.Scp
              (CallFunction|OnStartComicPanel))
            (:Event.HeroAbility.Dmg
              (CallFunction|OnAssignDamage)))))
      (Ability|EndAbility))
    (:Is%20Not%20Valid
      (Ability|EndAbility))))

(event Custom|OnStartComicPanel
  (Development|PrintString "@@@ Event OnStartComicPanel [GA_PistolCanary_Heroes_Ability_Parent]: CALLED!" false (Variables|Debug|GetbDebugAbility)))

(event Custom|OnAssignDamage
  (Development|PrintString "@@@ Event OnAssignDamage [GA_PistolCanary_Heroes_Ability_Parent]: CALLED!" false (Variables|Debug|GetbDebugAbility)))

(event Ability|EventOnEndAbility
  (params (bWasCancelled :type Boolean))
  (bind _aicontroller (Variables|Internal|GetAIController))
  (Development|PrintString "@@@ Event OnEndAbility [GA_PistolCanary_Heroes_Ability_Parent]: CALLED!" false (Variables|Debug|GetbDebugAbility))
  (Utilities|IsValid _aicontroller
    (:Is%20Valid
      (AI|PopBehaviorControl _aicontroller))))

(fn InitHeroSuperAbility
  (bind _asbp_pawn_pistol_canary_heroes (Utilities|Casting|CastToBP_Pawn_PistolCanary_Heroes (Ability|GetAvatarActorfromActorInfo))
    (:then
      (bind _output_get (Variables|Internal|SetBaseHeroPawn _asbp_pawn_pistol_canary_heroes))
      (bind _asfort_athena_aibot_controller (Utilities|Casting|CastToFortAthenaAIBotController (Pawn|GetController _output_get))
        (:then
          (Variables|Internal|SetAIController _asfort_athena_aibot_controller))
        (:CastFailed
          _)))
    (:CastFailed
      _))
  (Ability|EndAbility))

(fn SetupHeroSuperAbility
  (bind _aicontroller (Variables|Internal|GetAIController))
  (bind _returnvalue (Ability|CommitAbility))
  (if _returnvalue
    (AI|PushBehaviorControl _aicontroller -4194304)
    (Class|FortAthenaAIBotController|StopFire _aicontroller)
    (else
      (Ability|EndAbility))))
