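; Parent gameplay ability that plays recurring speech bubbles on PistolCanary hero AI pawns.
; On activation it caches the hero pawn, ending if the cast fails, then starts a self-retriggering loop with random delays from scalable min/max values.
; Each pass plays the moment-of-calm or attacking bubble per the pawn's flag, or ends if the pawn is dying, dead or speech-disabled.
; A target-elimination event replaces the attacking tag with the elimination tag.
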(event Ability|EventActivateAbility
  (|CallParentFunction)
  (bind _asbp_pawn_pistol_canary_heroes (Utilities|Casting|CastToBP_Pawn_PistolCanary_Heroes (Variables|InternalData|GetAIPawn))
    (:then
      (Variables|Default|SetHeroPawn _asbp_pawn_pistol_canary_heroes)
      (Utilities|FlowControl|Sequence
        (:then_0
          (Ability|Async|WaitGameplayEventToActor (Variables|InternalData|GetAIPawn) (Variables|Default|GetEventTag_OnTargetElimination) true
            (:EventReceived
              (Variables|Default|SetSpeechBubbleTag_Attacking (Variables|Default|GetSpeechBubbleTag_OnTargetElim)))))
        (:then_1
          (CallFunction|TriggerSpeechfromDelay))))
    (:CastFailed
      (Ability|EndAbility))))

(event Custom|TriggerSpeechFromDelay
  (Utilities|FlowControl|Delay (Math|Random|RandomFloatinRange (ScalableFloat|GetValueatLevel (Variables|Default|GetMinTimeBetweenSpeech)) (ScalableFloat|GetValueatLevel (Variables|Default|GetMaxTimeBetweenSpeech))))
  (if (Math|Boolean|ORBoolean (Variables|/Script/FortniteGame.FortPawn|GetbIsDying (Variables|InternalData|GetAIPawn)) (FortPawn|IsDead (Variables|InternalData|GetAIPawn)) (not (Variables|/PistolCanary/NPCs/Heroes/Pawns/BP_Pawn_PistolCanary_Heroes.BP_Pawn_PistolCanary_Heroes_C|GetbSpeechBubblesEnabled (Variables|Default|GetHeroPawn))))
    (Ability|EndAbility)
    (else
      (Class|BPFLAICommonFunctions|PlaySpeechBubble (Variables|InternalData|GetAIPawn) (select (Variables|/PistolCanary/NPCs/Heroes/Pawns/BP_Pawn_PistolCanary_Heroes.BP_Pawn_PistolCanary_Heroes_C|GetbSpeechIsMomentOfCalm (Variables|Default|GetHeroPawn)) (Variables|Default|GetSpeechBubbleTag_MomentOfCalm) (Variables|Default|GetSpeechBubbleTag_Attacking)))
      (CallFunction|TriggerSpeechfromDelay))))

(event Ability|EventFailedToActivatePassiveAbility
  (params (ActorInfo :type Gameplay%20Ability%20Actor%20Info%20Structure)))
