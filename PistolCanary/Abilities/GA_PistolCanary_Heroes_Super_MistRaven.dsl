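; Gameplay ability for the MistRaven hero's Super in PistolCanary, acting as a thin child of a parent ability class.
; It overrides the OnStartComicPanel and OnAssignDamage events, each calling the parent implementation and then printing a trace message.
; The traces stay off screen and reach the log only when bDebugAbility is set.
; The graph adds no other behavior of its own.
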(event EventOnStartComicPanel
  (|CallParentFunction)
  (Development|PrintString "@@@@@@ Event OnStartComicPanel [GA_PistolCanary_Heroes_Super_MistRaven]: CALLED!" false (Variables|Debug|GetbDebugAbility)))

(event EventOnAssignDamage
  (|CallParentFunction)
  (Development|PrintString "@@@@@@ Event OnAssignDamage [GA_PistolCanary_Heroes_Super_MistRaven]: CALLED!" false (Variables|Debug|GetbDebugAbility)))
